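# Sound cue for the DogSweater item's dissolve warning, a spatialized one-shot in the gadget items sound class with its own attenuation override.
# A quality-level node randomly picks between two warning waves in its first two slots, while the other two use a mobile-optimized wave with boosted volume.
# An enveloper ramps pitch and volume up over roughly the first third of a second, then a root modulator adds default random variation.
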SoundCue /Game/Athena/Sounds/Items/DogSweater/Athena_DogSweater_Dissolve_Warning_Cue
root=Modulator_0
nodes [
 {
  id=Enveloper_1
  children [QualityLevel_0]
  properties {pitchCurve {editorCurveData {keys [<at 0.0062 0.4928>,<at 0.2839 0.9536>]}}, volumeCurve {editorCurveData {keys [<at 0 0.1>,<at 0.35 1>]}}}
 }
 {id=Modulator_0, children [Enveloper_1]}
 {id=Modulator_1, children [WavePlayer_1], properties {pitchMax=1, pitchMin=1, volumeMax=2, volumeMin=2}}
 {id=QualityLevel_0, children [Random_1,Random_1,Modulator_1,Modulator_1]}
 {id=Random_1, children [WavePlayer_0,WavePlayer_3], properties {weights [1,1]}}
 {id=WavePlayer_0, properties {soundWaveAssetPtr=&/Athena_DogSweater_Warning_01}}
 {id=WavePlayer_1, properties {soundWaveAssetPtr=/Game/Athena/Sounds/Items/DogSweater/OPT/Mobile/Athena_DogSweater_Warning_01_OPT}}
 {id=WavePlayer_3, properties {soundWaveAssetPtr=&/Athena_DogSweater_Warning_02}}
]
properties {
 bOverrideAttenuation=true
 maxDistance=6400
 soundClassObject=&Fort_Gadgets_Items
 volumeMultiplier=0.25
 attenuationOverrides {dBAttenuationAtMax=-100, distanceAlgorithm=NaturalSound, falloffDistance=4000, attenuationShapeExtents {x=4000}}
 concurrencyOverrides {maxCount=2, volumeScaleAttackTime=0.3}
 concurrencySet [null]
}
