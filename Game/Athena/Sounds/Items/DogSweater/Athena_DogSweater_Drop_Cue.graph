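# Sound cue that plays when a Dog Sweater item is dropped in Athena, using weapon-drop attenuation and the loot-collect sound class at reduced volume.
# A quality-level switch at the root sends its first two levels to the standard drop wave.
# The last two levels play a mobile-optimized wave through a modulator that doubles its volume with fixed pitch.
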SoundCue /Game/Athena/Sounds/Items/DogSweater/Athena_DogSweater_Drop_Cue
root=QualityLevel_0
nodes [
 {id=Modulator_1, children [WavePlayer_1], properties {pitchMax=1, pitchMin=1, volumeMax=2, volumeMin=2}}
 {id=QualityLevel_0, children [WavePlayer_0,WavePlayer_0,Modulator_1,Modulator_1]}
 {id=WavePlayer_0, properties {soundWaveAssetPtr=&/Athena_DogSweater_Drop_01}}
 {id=WavePlayer_1, properties {soundWaveAssetPtr=/Game/Athena/Sounds/Items/DogSweater/OPT/Mobile/Athena_DogSweater_Drop_01_OPT}}
]
properties {
 attenuationSettings=&Fort_Weapon_Drop_Att
 maxDistance=3000
 soundClassObject=&Fort_Loot_Collect
 volumeMultiplier=0.25
}
