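# Sound Cue for the looping grow sound of the Athena DogSweater item, routed to the melee weapon sound class.
# A Quality Level node feeds the standard looping wave to its first two slots and a mobile-optimized wave, with doubled volume via a Modulator, to the other two.
# An Enveloper at the root ramps pitch and volume up from a low start.
# Overrides set long-range attenuation and limit concurrency to two instances.
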SoundCue /Game/Athena/Sounds/Items/DogSweater/Athena_DogSweater_Grow_Loop_Cue
root=Enveloper_1
nodes [
 {
  id=Enveloper_1
  children [QualityLevel_0]
  properties {pitchCurve {editorCurveData {keys [<at 0.0062 0.4928>, {time=1}]}}, volumeCurve {editorCurveData {keys [<at 0 0.1>,<at 5 0.5>]}}}
 }
 {id=Modulator_1, children [WavePlayer_2], properties {pitchMax=1, pitchMin=1, volumeMax=2, volumeMin=2}}
 {id=QualityLevel_0, children [WavePlayer_0,WavePlayer_0,Modulator_1,Modulator_1]}
 {id=WavePlayer_0, properties {bLooping=true, soundWaveAssetPtr=&/Athena_DogSweater_Grow_Loop}}
 {
  id=WavePlayer_2
  properties {bLooping=true, soundWaveAssetPtr=/Game/Athena/Sounds/Items/DogSweater/OPT/Mobile/Athena_DogSweater_Grow_Loop_OPT}
 }
]
properties {
 bOverrideAttenuation=true
 bOverrideConcurrency=true
 maxDistance=20000
 soundClassObject=&Fort_WeaponMelee
 volumeMultiplier=0.2
 attenuationOverrides {falloffDistance=20000, attenuationShapeExtents {x=0}}
 concurrencyOverrides {maxCount=2, volumeScale=0.5, volumeScaleAttackTime=1}
 concurrencySet [null]
}
