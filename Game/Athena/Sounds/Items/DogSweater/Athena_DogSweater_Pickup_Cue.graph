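# Sound cue for the Dog Sweater item pickup in Athena, using weapon pickup attenuation and the gadget items sound class.
# A root Modulator with default settings feeds a quality-level switch.
# Its first two slots randomly pick between two equally weighted pickup waves.
# Its last two play a single optimized mobile wave through a Modulator that fixes pitch and doubles volume.
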SoundCue /Game/Athena/Sounds/Items/DogSweater/Athena_DogSweater_Pickup_Cue
root=Modulator_0
nodes [
 {id=Modulator_0, children [QualityLevel_0]}
 {id=Modulator_1, children [WavePlayer_0], properties {pitchMax=1, pitchMin=1, volumeMax=2, volumeMin=2}}
 {id=QualityLevel_0, children [Random_0,Random_0,Modulator_1,Modulator_1]}
 {id=Random_0, children [WavePlayer_1,WavePlayer_2], properties {weights [1,1]}}
 {id=WavePlayer_0, properties {soundWaveAssetPtr=/Game/Athena/Sounds/Items/DogSweater/OPT/Mobile/Athena_DogSweater_Pickup_02_OPT}}
 {id=WavePlayer_1, properties {soundWaveAssetPtr=&/Athena_DogSweater_Pickup_01}}
 {id=WavePlayer_2, properties {soundWaveAssetPtr=&/Athena_DogSweater_Pickup_02}}
]
properties {
 attenuationSettings=&Fort_Weapon_Pickup_Att
 bOverrideConcurrency=true
 maxDistance=3500
 soundClassObject=&Fort_Gadgets_Items
 volumeMultiplier=0.25
}
