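# Sound cue that plays the looping in-air sound of the DogSweater item projectile in Athena.
# A quality-level root sends its first two tiers straight to a looping wave player.
# The other two tiers pass a looping player of the same-named optimized wave through a modulator that doubles volume at fixed pitch.
# It uses the ranged projectile effects sound class and a close projectile attenuation asset.
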SoundCue /Game/Athena/Sounds/Items/DogSweater/Athena_DogSweater_Projectile_InAir_Loop_Cue
root=QualityLevel_0
nodes [
 {id=Modulator_2, children [WavePlayer_3], properties {pitchMax=1, pitchMin=1, volumeMax=2, volumeMin=2}}
 {id=QualityLevel_0, children [WavePlayer_0,WavePlayer_0,Modulator_2,Modulator_2]}
 {
  id=WavePlayer_0
  properties {bLooping=true, soundWaveAssetPtr=/Game/Athena/Sounds/Items/DogSweater/OPT/Mobile/Athena_DogSweater_Projectile_InAir_Loop_02_OPT}
 }
 {
  id=WavePlayer_3
  properties {bLooping=true, soundWaveAssetPtr=&Athena_DogSweater_Projectile_InAir_Loop_02_OPT}
 }
]
properties {
 attenuationSettings=&Fort_Weapon_Projectile_Close_Athena_Att
 maxDistance=4950
 soundClassObject=&Fort_WeaponRanged_Projectile_Effects
 volumeMultiplier=1.45
 attenuationOverrides {bAttenuate=false, falloffDistance=6000, attenuationShapeExtents {x=1000}}
}
