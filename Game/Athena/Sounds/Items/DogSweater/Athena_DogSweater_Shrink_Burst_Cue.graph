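# Sound cue for the shrink burst of the DogSweater item, in the gadget items sound class.
# A quality switch plays a random pick of two burst waves on the upper two tiers and one optimized mobile wave with boosted volume on the lower two.
# An Enveloper at the root fades volume in and shapes pitch over time.
# Overrides set natural-sound distance falloff and limit playback to one concurrent instance.
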SoundCue /Game/Athena/Sounds/Items/DogSweater/Athena_DogSweater_Shrink_Burst_Cue
root=Enveloper_0
nodes [
 {
  id=Enveloper_0
  children [QualityLevel_0]
  properties {pitchCurve {editorCurveData {keys [<at 0 0.868>, {time=1}]}}, volumeCurve {editorCurveData {keys [<at 0.0015 0.1516>,<at 0.5 1>]}}}
 }
 {id=Modulator_2, children [WavePlayer_0], properties {pitchMax=1, pitchMin=1, volumeMax=2, volumeMin=2}}
 {id=QualityLevel_0, children [Random_0,Random_0,Modulator_2,Modulator_2]}
 {id=Random_0, children [WavePlayer_2,WavePlayer_3], properties {weights [1,1]}}
 {id=WavePlayer_0, properties {soundWaveAssetPtr=/Game/Athena/Sounds/Items/DogSweater/OPT/Mobile/Athena_DogSweater_Shrink_Burst_02_OPT}}
 {id=WavePlayer_2, properties {soundWaveAssetPtr=&/Athena_DogSweater_Shrink_Burst_02}}
 {id=WavePlayer_3, properties {soundWaveAssetPtr=&/Athena_DogSweater_Shrink_Burst_01}}
]
properties {
 bOverrideAttenuation=true
 bOverrideConcurrency=true
 maxDistance=6500
 soundClassObject=&Fort_Gadgets_Items
 volumeMultiplier=0.4
 attenuationOverrides {dBAttenuationAtMax=-120, distanceAlgorithm=NaturalSound, falloffDistance=6000, attenuationShapeExtents {x=3500}}
 concurrencyOverrides {maxCount=1, volumeScaleAttackTime=0.3}
 concurrencySet [null]
}
