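# Looping sound cue for the shrink effect of the Athena DogSweater item, routed through the gadget items sound class.
# A quality-level node picks either the standard looping wave or a mobile-optimized wave whose volume a Modulator doubles.
# An Enveloper at the root fades volume up over about three seconds and applies a one-second pitch curve.
# Overrides limit playback to one instance and set a long attenuation falloff.
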SoundCue /Game/Athena/Sounds/Items/DogSweater/Athena_DogSweater_Shrink_Loop_Cue
root=Enveloper_1
nodes [
 {
  id=Enveloper_1
  children [QualityLevel_0]
  properties {pitchCurve {editorCurveData {keys [<at 0 0.868>, {time=1}]}}, volumeCurve {editorCurveData {keys [<at 0.0015 0.1516>,<at 3 1>]}}}
 }
 {id=Modulator_2, children [WavePlayer_1], properties {pitchMax=1, pitchMin=1, volumeMax=2, volumeMin=2}}
 {id=QualityLevel_0, children [WavePlayer_0,WavePlayer_0,Modulator_2,Modulator_2]}
 {id=WavePlayer_0, properties {bLooping=true, soundWaveAssetPtr=&/Athena_DogSweater_Shrink_Loop}}
 {
  id=WavePlayer_1
  properties {bLooping=true, soundWaveAssetPtr=/Game/Athena/Sounds/Items/DogSweater/OPT/Mobile/Athena_DogSweater_Shrink_Loop_OPT}
 }
]
properties {
 bOverrideAttenuation=true
 bOverrideConcurrency=true
 maxDistance=20000
 soundClassObject=&Fort_Gadgets_Items
 volumeMultiplier=0.4
 attenuationOverrides {falloffDistance=20000, attenuationShapeExtents {x=0}}
 concurrencyOverrides {maxCount=1, volumeScaleAttackTime=0.3}
 concurrencySet [null]
}
