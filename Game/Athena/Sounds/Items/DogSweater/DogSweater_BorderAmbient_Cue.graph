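# Looping ambient cue for the DogSweater item's border, routed to the gadget items sound class.
# An Enveloper fades the volume in over five seconds on top of a quality-level switch.
# The first two quality slots play the full loop, and the other two play the optimized mobile loop through a Modulator that doubles its volume.
# Overrides give it long-range attenuation and limit concurrent instances with strong volume scaling.
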SoundCue /Game/Athena/Sounds/Items/DogSweater/DogSweater_BorderAmbient_Cue
root=Enveloper_0
nodes [
 {
  id=Enveloper_0
  children [QualityLevel_0]
  properties {pitchCurve {editorCurveData {keys [{time=0}]}}, volumeCurve {editorCurveData {keys [<at 0 0>,<at 5 1>]}}}
 }
 {id=Modulator_2, children [WavePlayer_0], properties {pitchMax=1, pitchMin=1, volumeMax=2, volumeMin=2}}
 {id=QualityLevel_0, children [WavePlayer_2,WavePlayer_2,Modulator_2,Modulator_2]}
 {
  id=WavePlayer_0
  properties {bLooping=true, soundWaveAssetPtr=/Game/Athena/Sounds/Items/DogSweater/OPT/Mobile/Athena_DogSweater_BorderAmbient_Loop_OPT}
 }
 {id=WavePlayer_2, properties {bLooping=true, soundWaveAssetPtr=&/Athena_DogSweater_BorderAmbient_Loop}}
]
properties {
 bOverrideAttenuation=true
 bOverrideConcurrency=true
 maxDistance=21000
 soundClassObject=&Fort_Gadgets_Items
 volumeMultiplier=0.65
 attenuationOverrides {falloffDistance=20000, attenuationShapeExtents {x=1000}}
 concurrencyOverrides {bVolumeScaleCanRelease=true, maxCount=3, volumeScale=0.05, volumeScaleAttackTime=0.3, volumeScaleReleaseTime=2}
 concurrencySet [null]
}
